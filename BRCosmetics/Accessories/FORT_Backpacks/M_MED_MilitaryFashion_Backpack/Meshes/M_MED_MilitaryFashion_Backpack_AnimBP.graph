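# Anim Blueprint for the MilitaryFashion backpack that adds secondary jiggle to its dangling gear and antenna.
# Input pose gets an additive OffsetAll nudge, then chained AnimDynamics nodes drive flashlight, canteen, walkie, antenna, battery and top handle.
# The sprung first antenna bone uses inverted gravity and copies its rotation delta to antenna2 and antenna3, leaving antenna2's own dynamics node unconnected.
# The event graph lowers only the battery's alpha while targeting.
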AnimBlueprint /BRCosmetics/Accessories/FORT_Backpacks/M_MED_MilitaryFashion_Backpack/Meshes/M_MED_MilitaryFashion_Backpack_AnimBP
target_skeleton=&/M_MED_MilitaryFashion_Backpack_Skeleton
graphs [
 AnimGraph {nodes [
   Anim Dynamics - Bone: dyn_antenna2(Alpha=0.75, ExternalForce="0,0,0") {
    angularDampingOverride=0.999
    angularSpringConstant=4800
    bAngularSpring=true
    bGravityOverrideInSimSpace=true
    bOverrideAngularDamping=true
    bUseGravityOverride=true
    alphaBoolBlend {blendOption=Linear}
    boundBone {boneName=dyn_antenna2}
    componentAppliedLinearAccClamp=<50 50 50>
    componentLinearAccScale=<1 1 1>
    componentLinearVelScale=<1 1 1>
    gravityOverride {z=981}
    physicsBodyDefinitions [{
      sphereCollisionRadius=0
      boundBone {boneName=dyn_antenna2}
      boxExtents=<6 8 8>
      constraintSetup {twistAxis=AxisZ, angularLimitsMax=<0 12 12>, angularLimitsMin=<0 -12 -3>, angularTarget=<0 0 0>}
      localJointOffset=<2 0 0.2>
     }]
   }
   AnimGraphNode_SubInput_0 = Input Pose()
   n1 = "Transform (Modify) Bone - Bone: OffsetAll"!(ComponentPose=Local To Component(LocalPose=@AnimGraphNode_SubInput_0), Alpha=1, Translation="1,0,0", Rotation="-8.000,0.000,0.000") -> (Pose) {
    rotationMode=BMM_Additive
    translationMode=BMM_Additive
    alphaBoolBlend {blendOption=Linear}
    boneToModify {boneName=OffsetAll}
    rotation {pitch=-8}
    translation {x=1}
   }
   n2 = Anim Dynamics - Bone: dyn_flashlight(ComponentPose=@n1, Alpha=0.2, ExternalForce="0,0,0") {
    angularDampingOverride=0.995
    angularSpringConstant=300
    bGravityOverrideInSimSpace=true
    bOverrideAngularDamping=true
    bUseGravityOverride=true
    alphaBoolBlend {blendOption=Linear}
    boundBone {boneName=dyn_flashlight}
    componentAppliedLinearAccClamp=<500 500 500>
    componentLinearAccScale=<1 1 1>
    componentLinearVelScale=<1 1 1>
    gravityOverride {z=-981}
    physicsBodyDefinitions [{
      sphereCollisionRadius=0
      boundBone {boneName=dyn_flashlight}
      boxExtents=<6 6 2>
      constraintSetup {twistAxis=AxisZ, angularLimitsMax=<3 10 3>, angularLimitsMin=<-3 -1 -3>, angularTarget=<0 0 0>}
      localJointOffset=<0 0 -2>
     }]
   }
   n3 = Anim Dynamics - Bone: dyn_canteen(ComponentPose=@n2, Alpha=0.2, ExternalForce="0,0,0") {
    angularDampingOverride=0.995
    angularSpringConstant=300
    bGravityOverrideInSimSpace=true
    bOverrideAngularDamping=true
    bUseGravityOverride=true
    alphaBoolBlend {blendOption=Linear}
    boundBone {boneName=dyn_canteen}
    componentAppliedLinearAccClamp=<500 500 500>
    componentLinearAccScale=<1 1 1>
    componentLinearVelScale=<1 1 1>
    gravityOverride {z=-981}
    physicsBodyDefinitions [{
      sphereCollisionRadius=0
      boundBone {boneName=dyn_canteen}
      boxExtents=<12 12 4>
      constraintSetup {twistAxis=AxisZ, angularLimitsMax=<3 3 3>, angularLimitsMin=<-3 -3 -3>, angularTarget=<0 0 0>}
      localJointOffset=<0 0 -3>
     }]
   }
   n4 = Anim Dynamics - Bone: dyn_walkie(ComponentPose=@n3, Alpha=0.2, ExternalForce="0,0,0") {
    angularDampingOverride=0.995
    angularSpringConstant=300
    bGravityOverrideInSimSpace=true
    bOverrideAngularDamping=true
    bUseGravityOverride=true
    alphaBoolBlend {blendOption=Linear}
    boundBone {boneName=dyn_walkie}
    componentAppliedLinearAccClamp=<500 500 500>
    componentLinearAccScale=<1 1 1>
    componentLinearVelScale=<1 1 1>
    gravityOverride {z=-981}
    physicsBodyDefinitions [{
      sphereCollisionRadius=0
      boundBone {boneName=dyn_walkie}
      boxExtents=<16 16 5>
      constraintSetup {twistAxis=AxisZ, angularLimitsMax=<1 3 3>, angularLimitsMin=<-1 -3 -3>, angularTarget=<0 0 0>}
      localJointOffset=<0 0 -4>
     }]
   }
   n5 = Anim Dynamics - Bone: dyn_antenna1(ComponentPose=@n4, Alpha=0.5, ExternalForce="0,0,0") {
    angularDampingOverride=0.999
    angularSpringConstant=800
    bAngularSpring=true
    bGravityOverrideInSimSpace=true
    bOverrideAngularDamping=true
    bUseGravityOverride=true
    alphaBoolBlend {blendOption=Linear}
    boundBone {boneName=dyn_antenna1}
    componentAppliedLinearAccClamp=<50 50 50>
    componentLinearAccScale=<1 1 1>
    componentLinearVelScale=<1 1 1>
    gravityOverride {z=981}
    physicsBodyDefinitions [{
      sphereCollisionRadius=0
      boundBone {boneName=dyn_antenna1}
      boxExtents=<6 8 8>
      constraintSetup {angularLimitsMax=<0 12 18>, angularLimitsMin=<0 -12 -6>, angularTarget=<0 0 0>}
      localJointOffset=<2 0 0.1>
     }]
   }
   n6 = Copy Bone Delta - Source Bone: dyn_antenna1 - Target Bone: dyn_antenna2!(ComponentPose:Component Space Pose Link Structure=@n5, Alpha:float=1) -> (Pose:Component Space Pose Link Structure) {
    is=title AnimGraphNode_CopyBoneDelta
    bCopyRotation=true
    alphaBoolBlend {blendOption=Linear}
    sourceBone {boneName=dyn_antenna1}
    targetBone {boneName=dyn_antenna2}
   }
   n7 = Copy Bone Delta - Source Bone: dyn_antenna1 - Target Bone: dyn_antenna3!(ComponentPose:Component Space Pose Link Structure=@n6, Alpha:float=1) -> (Pose:Component Space Pose Link Structure) {
    is=title AnimGraphNode_CopyBoneDelta
    bCopyRotation=true
    alphaBoolBlend {blendOption=Linear}
    sourceBone {boneName=dyn_antenna1}
    targetBone {boneName=dyn_antenna3}
   }
   n8 = Anim Dynamics - Bone: dyn_battery(ComponentPose=@n7, Alpha=Get TargetAlpha() -> (TargetAlpha:double), ExternalForce="0,0,0") {
    angularDampingOverride=0.995
    angularSpringConstant=300
    bGravityOverrideInSimSpace=true
    bOverrideAngularDamping=true
    bUseGravityOverride=true
    alphaBoolBlend {blendOption=Linear}
    boundBone {boneName=dyn_battery}
    componentAppliedLinearAccClamp=<500 500 500>
    componentLinearAccScale=<1 1 1>
    componentLinearVelScale=<1 1 1>
    gravityOverride {z=-981}
    physicsBodyDefinitions [{
      sphereCollisionRadius=0
      boundBone {boneName=dyn_battery}
      boxExtents=<16 16 5>
      constraintSetup {twistAxis=AxisZ, angularLimitsMax=<3 1 3>, angularLimitsMin=<-3 -1 -3>, angularTarget=<0 0 0>}
      localJointOffset=<0 0 -4>
     }]
   }
   n9 = Anim Dynamics - Bone: dyn_topHandle(ComponentPose=@n8, Alpha=0.6, ExternalForce="0,0,0") {
    angularDampingOverride=0.995
    angularSpringConstant=600
    bGravityOverrideInSimSpace=true
    bOverrideAngularDamping=true
    bUseGravityOverride=true
    alphaBoolBlend {blendOption=Linear}
    boundBone {boneName=dyn_topHandle}
    componentAppliedLinearAccClamp=<200 200 200>
    componentLinearAccScale=<1 1 1>
    componentLinearVelScale=<1 1 1>
    gravityOverride {z=-9.81}
    physicsBodyDefinitions [{
      sphereCollisionRadius=0
      boundBone {boneName=dyn_topHandle}
      boxExtents=<12 12 6>
      constraintSetup {twistAxis=AxisZ, angularLimitsMax=<3 3 3>, angularLimitsMin=<-3 -3 -3>, angularTarget=<0 0 0>}
      localJointOffset=<0 0 -8>
     }]
   }
   Output Pose(Result=Component To Local(ComponentPose=@n9)) {layerGroup=None}
  ]}
 EventGraph {nodes [
   Event BlueprintUpdateAnimation()
   > Set TargetAlpha(TargetAlpha:double=SelectFloat(A=0.1, B=0.2, bPickA=Get bIsTargeting())) -> (Output_Get:double)
  ]}
]
variables [TargetAlpha {type=number}]
